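# Translucent unlit liquid ribbon material for Niagara mesh particles in the Sunrose Pool quest-item effect.
# SplineThicken takes its width from a dynamic-parameter Thickness times a sphere mask on UV V, with RevealPosition driving both mask center and radius.
# Panned chrome-liquid normals feed SplineThicken, whose view-space normal samples a fake reflection texture that is scaled by RefBrightness and combined with particle color.
# The panned liquid mask is multiplied by particle alpha and by the reveal mask.
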Material /BattlePassS29_Quests/Items/FX/SunrosePool/M_FX_SunrosePool_LiquidRibbon
graph {
 properties {
  bTangentSpaceNormal=false
  bUsedWithNiagaraMeshParticles=true
  blendMode=BLEND_Translucent
  pixelDepthOffsetMode=PDOM_Legacy
  shadingModel=MSM_Unlit
  usedShadingModels=Unlit
 }
 nodes [
  n1 = ParticleColor()
  n2 = DynamicParameter() {paramNames [Thickness, RevealPosition, RevealRadius, RefBrightness]}
  n3 = SphereMask(A=ComponentMask(None=TextureCoordinate()) {g=true}, B=@n2.RevealPosition, Radius=@n2.RevealPosition) {hardnessPercent=0}
  n4 = Multiply(A=@n3, B=@n2.Thickness)
  n5 = Panner(
   Coordinate=TextureCoordinate() {uTiling=0.25, vTiling=0.5}
   Time=DynamicParameter() {parameterIndex=1, paramNames [Thickness/Tiling, RevealPosition/Speed, RevealRadius, Param4]}.RevealPosition/Speed
  ) {speedY=0.2}
  n6 = MaterialFunctionCall(
   WidthBase=@n4
   WidthTip=@n4
   AdditionalNormal=TextureSample(UVs=@n5) {samplerType=SAMPLERTYPE_Normal, texture=/Game/Effects/Fort_Effects/Textures/Water/T_ChromeLiquid_4x2_N}.RGB
  ) {materialFunction=&SplineThicken}
  n7 = Multiply(A=ComponentMask(None=Transform(None=@n6.Normal) {transformSourceType=TRANSFORMSOURCE_World, transformType=TRANSFORM_View}) {g=true, r=true}) {constB=0.5}
  n8 = TextureSample(UVs=Multiply(A=Add(A=@n7) {constB=0.5})) {texture=/Game/VisualThreatManager/StormVisuals/LargeClouds/T_EyeballReflection2}
  Add(A=Multiply(A=@n1.RGB), B=Add(A=@n1.RGB, B=Multiply(A=@n8.RGB, B=@n2.RefBrightness) {constB=2}))
  Max()
  n9 = TextureSample(UVs=@n5) {texture=/Game/Effects/Fort_Effects/Textures/Water/T_ChromeLiquid_4x2_Mask}
  Multiply(A=ParticleColor().A, B=@n9.R)
  Multiply(A=@n9.R, B=@n3)
  TextureCoordinate()
  TextureSample() {texture=/Game/Effects/Fort_Effects/Gameplay/SmokeDustExplosions/Textures/T_RibbonSmokeMasked_04}
  TextureSample() {samplerType=SAMPLERTYPE_Normal, texture=/Game/Effects/Fort_Effects/Gameplay/SmokeDustExplosions/Textures/T_RibbonSmokeMasked_01}
 ]
}
