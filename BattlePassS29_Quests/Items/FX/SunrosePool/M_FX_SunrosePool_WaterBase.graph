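# Translucent, two-sided unlit water material for Niagara mesh particles in the Sunrose Pool item effect.
# SplineThicken, driven by a dynamic Thickness parameter, reshapes the mesh, with panned noise adding vertical displacement.
# A particle-driven panning chrome-liquid mask combines with particle alpha and a sphere mask for shape.
# The matching normal map drives fake matcap-style reflections via view-space lookup, plus Fresnel shading alongside inner, edge and highlight colors.
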Material /BattlePassS29_Quests/Items/FX/SunrosePool/M_FX_SunrosePool_WaterBase
graph {
 properties {
  bTangentSpaceNormal=false
  bUsedWithNiagaraMeshParticles=true
  blendMode=BLEND_Translucent
  pixelDepthOffsetMode=PDOM_Legacy
  previewMesh=&/FX_SM_SunrosePool_Source
  shadingModel=MSM_Unlit
  twoSided=true
  usedShadingModels=Unlit
 }
 nodes [
  n1 = DynamicParameter() {paramNames [Pan, OffsetX, Thickness, ReflectionsTiling]}
  n2 = MaterialFunctionCall(WidthBase=@n1.Thickness, WidthTip=@n1.Thickness) {materialFunction=&SplineThicken}
  n3 = Panner(Coordinate=TextureCoordinate() {uTiling=0.25, vTiling=0.25}) {speedY=1.5}
  n4 = Subtract(A=Multiply(A=TextureSample(UVs=@n3) {texture=/Game/Effects/Fort_Effects/Textures/Tile/T_GaussianNoise_01_M}.R) {constB=2})
  n5 = ScalarParameter() {parameterName=Noise_VS_Str}
  Add(A=@n2.WorldPositionOffset, B=Multiply(A=Constant3Vector() {constant=<rgba 0 0 1 0>}, B=Multiply(A=@n4, B=@n5)))
  n6 = TextureCoordinate()
  n7 = ComponentMask(None=VectorParameter() {parameterName=BaseTiling, defaultValue=<rgba 4 0.25 0 1>}.RGB) {g=true, r=true}
  n8 = DynamicParameter() {paramNames [Pan, OffsetX, Thickness, ReflectionsTiling]}
  n9 = Constant()
  n10 = Panner(
   Coordinate=Add(A=Multiply(A=@n6, B=@n7), B=AppendVector(A=@n8.OffsetX, B=@n9))
   Time=DynamicParameter() {paramNames [Pan, OffsetX, Thickness, ReflectionsTiling]}.Pan
  ) {speedY=1}
  Multiply(
   A=Multiply(
    A=ParticleColor().A
    B=TextureSample(UVs=@n10) {samplerSource=SSM_Clamp_WorldGroupSettings, texture=/Game/Effects/Fort_Effects/Textures/Water/T_ChromeLiquid_4x2_Mask}.G
   )
   B=SphereMask(A=TextureCoordinate(), B=Constant() {r=0.5}) {attenuationRadius=0.5, hardnessPercent=0}
  )
  n11 = VectorParameter() {parameterName=HighlightColor, defaultValue=<rgba 0 0 0 1>}
  n12 = Multiply(A=Multiply(A=@n11.RGB, B=@n11.RGB), B=NamedRerouteUsage())
  n13 = LinearInterpolate(
   A=VectorParameter() {parameterName=InnerColor, defaultValue=<rgba 0 1 0 1>}.RGB
   B=VectorParameter() {parameterName=EdgeColor, defaultValue=<rgba 1 0 0 1>}.RGB
   Alpha=NamedRerouteUsage()
  )
  Multiply(A=Add(A=Multiply(A=@n12, B=@n13), B=@n13), B=ComponentMask(None=TextureCoordinate()) {r=true})
  n14 = TextureSample(UVs=@n10) {
   samplerSource=SSM_Clamp_WorldGroupSettings
   samplerType=SAMPLERTYPE_Normal
   texture=/Game/Effects/Fort_Effects/Textures/Water/T_ChromeLiquid_4x2_N
  }
  n15 = Transform(None=@n14.RGB) {transformType=TRANSFORM_View}
  n16 = Add(A=Multiply(A=ComponentMask(None=@n15) {g=true, r=true}) {constB=0.5}) {constB=0.5}
  NamedRerouteDeclaration(None=TextureSample(UVs=Multiply(A=@n16) {constB=0.5}) {texture=/Game/Packages/Fortress_SharedMaterials/Textures/T_Ravens_SpecularReflection}.RGB) {name=Highlights, nodeColor=<rgba 0.8516 1 0 1>}
  NamedRerouteDeclaration(None=MaterialFunctionCall(Exponent=ScalarParameter() {defaultValue=3, parameterName=FresnelExponent}, Vertex / Pixel Normal=Transform(None=@n14.RGB)) {materialFunction=&MF_Fresnel_FX}.Fresnel Inverted Double Sided) {name=ShadingFresnel, nodeColor=<rgba 0.5 0 1 1>}
  Saturate(None=DotProduct(A=@n2.Normal, B=Constant3Vector() {constant=<rgba 0 0 1 0>}))
  ScalarParameter() {parameterName=size}
  VertexNormalWS()
 ]
}
