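# Translucent unlit Niagara particle material for water elements in the Sunrose Pool effect.
# A panning, per-particle randomized noise normal combines with a sphere normal and transforms to view space to sample a fake specular reflection texture, tinted with particle color.
# The noise also distorts sphere-mask radii, shaping a wobbly round mask scaled by particle alpha and an inverted EdgeLight mask.
# Dynamic parameters control reflection strength and tiling, detail tiling and speed, and distortion.
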Material /BattlePassS29_Quests/Items/FX/SunrosePool/M_FX_SunrosePool_WaterElems
graph {
 properties {
  bGenerateSphericalParticleNormals=true
  bUsedWithNiagaraMeshParticles=true
  bUsedWithNiagaraSprites=true
  blendMode=BLEND_Translucent
  pixelDepthOffsetMode=PDOM_Legacy
  shadingModel=MSM_Unlit
  usedShadingModels=Unlit
 }
 nodes [
  n1 = ParticleRandom()
  n2 = DynamicParameter() {paramNames [Reflections, DetailTiling, DetailSpeed, ReflectionsTiling]}
  n3 = Multiply(A=TextureCoordinate() {uTiling=0.3, vTiling=0.3}, B=@n2.DetailTiling)
  n4 = ParticleRelativeTime()
  n5 = TextureSample(UVs=Panner(Coordinate=Add(A=@n1, B=@n3), Time=Multiply(A=@n4, B=@n2.DetailSpeed)) {speedY=1}) {samplerType=SAMPLERTYPE_Normal, texture=/Game/Effects/Fort_Effects/Textures/Tile/T_GaussianNoise_02_N}
  n6 = TextureSample() {samplerType=SAMPLERTYPE_Normal, texture=/Game/Effects/Fort_Effects/Textures/Mask/T_Sphere_N}
  n7 = Add(A=Multiply(A=@n5.RGB), B=Multiply(A=@n6.RGB, B=Constant3Vector() {constant=<rgba 1 1 -1 0>}))
  n8 = Multiply(A=ComponentMask(None=Transform(None=@n7) {transformType=TRANSFORM_View}) {g=true, r=true}) {constB=0.5}
  n9 = DynamicParameter() {paramNames [Reflections, DetailTiling, DetailSpeed, ReflectionsTiling]}
  n10 = TextureSample(UVs=Multiply(A=Add(A=@n8) {constB=0.5}, B=@n9.ReflectionsTiling)) {texture=/Game/Packages/Fortress_SharedMaterials/Textures/T_Ravens_SpecularReflection}
  Add(A=NamedRerouteUsage(), B=@n10.RGB)
  n11 = ParticleColor()
  Add(A=Multiply(A=@n11.RGB, B=NamedRerouteUsage()), B=Add(A=@n11.RGB, B=Multiply(A=@n9.Reflections, B=@n10.RGB)))
  Fresnel(ExponentIn=@n6.RGB)
  n12 = TextureCoordinate()
  n13 = Constant() {r=0.5}
  n14 = Add(A=Multiply(
    A=@n5.R
    B=DynamicParameter() {parameterIndex=1, paramNames [Reflections/DistStrength, DetailTiling, DetailSpeed, ReflectionsTiling]}.Reflections/DistStrength
   ) {constB=0.3}) {constB=0.4}
  Multiply(A=@n11.A, B=SphereMask(A=@n12, B=@n13, Radius=@n14) {attenuationRadius=0.47})
  NamedRerouteDeclaration(None=OneMinus(None=SphereMask(A=@n12, B=@n13, Radius=@n14) {hardnessPercent=50})) {name=EdgeLight, nodeColor=<rgba 0 0.8125 1 1>}
 ]
}
