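# Translucent unlit Niagara sprite and mesh particle material for flowing water sheets in the Sunrose Pool effect.
# Emissive combines gradient-shaded particle color with fake reflections from noise sampled along a panning-normal reflection vector.
# Opacity erodes an offsettable wipe gradient with panning splash and streak textures, shaped by end fade and vertex red.
# The streak texture also drives a vertex-normal wobble, and dynamic parameters control tiling, speed, offsets and opacity boost.
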Material /BattlePassS29_Quests/Items/FX/SunrosePool/M_FX_SunrosePool_WaterSheet
graph {
 properties {
  bUsedWithNiagaraMeshParticles=true
  bUsedWithNiagaraSprites=true
  blendMode=BLEND_Translucent
  pixelDepthOffsetMode=PDOM_Legacy
  shadingModel=MSM_Unlit
  usedShadingModels=Unlit
 }
 nodes [
  n1 = Saturate(None=ComponentMask(None=TextureCoordinate()) {g=true})
  n2 = Multiply(A=Add(A=Multiply(A=@n1) {constB=2}) {constB=0.5}, B=ParticleColor().RGB)
  n3 = OneMinus(None=NamedRerouteUsage())
  n4 = DynamicParameter() {paramNames [Reflections, DetailTiling, DetailSpeed, ReflectionsTiling]}
  n5 = TextureCoordinate() {vTiling=0.25}
  n6 = DynamicParameter() {paramNames [Reflections, DetailTiling, DetailSpeed, ReflectionsTiling]}
  n7 = Panner(Coordinate=Add(A=ParticleRandom(), B=Multiply(A=@n5, B=@n6.DetailTiling)), Time=Multiply(A=ParticleRelativeTime(), B=@n6.DetailSpeed)) {speedY=-1}
  n8 = ScalarParameter() {parameterName=DetailNrm}
  n9 = AppendVector(A=@n8, B=@n8)
  n10 = Constant()
  n11 = Add(
   A=Constant3Vector() {constant=<rgba 0 0 1 0>}
   B=Multiply(
    A=TextureSample(UVs=@n7) {samplerType=SAMPLERTYPE_Normal, texture=/Game/Effects/Fort_Effects/Textures/Tile/T_GaussianNoise_02_N}.RGB
    B=AppendVector(A=@n9, B=@n10)
   ) {constB=0}
  )
  n12 = MaterialFunctionCall!(CameraVector=MaterialFunctionCall!(Normal=Transform(None=@n11), CameraVector=CameraVectorWS()) {materialFunction=/Engine/Functions/Engine_MaterialFunctions02/Utility/CustomReflectionVector}.Reflection Vector) {materialFunction=/Engine/Functions/Engine_MaterialFunctions01/Texturing/CameraVectorToLatLongUV}
  n13 = Multiply(
   A=@n4.Reflections
   B=TextureSample(UVs=Multiply(A=@n12.Uvs, B=@n4.ReflectionsTiling) {constB=1.75}) {texture=/Game/Effects/Fort_Effects/Textures/Tile/T_TileNoise_01}.R
  )
  MaterialFunctionCall(Emissive=Add(A=Multiply(A=@n2, B=@n3), B=Add(A=@n2, B=@n13))) {materialFunction=&MF_FX_EmissiveAdjustments}
  n14 = DynamicParameter() {
   parameterIndex=1
   defaultValue=<rgba 0 1 1 1>
   paramNames [Reflections/MaskOffset, DetailTiling, DetailSpeed, ReflectionsTiling/DetailsTiling]
  }
  n15 = Add(A=Multiply(A=TextureCoordinate() {vTiling=2}, B=@n14.ReflectionsTiling/DetailsTiling), B=NamedRerouteUsage())
  n16 = ScalarParameter() {defaultValue=1, parameterName=VertexWobbleTexScale}
  n17 = Multiply(A=@n15, B=@n16)
  n18 = DynamicParameter() {
   parameterIndex=1
   defaultValue=<rgba 0 1 1 1>
   paramNames [Reflections/MaskOffset, DetailTiling, DetailSpeed, ReflectionsTiling/DetailsTiling]
  }
  n19 = Multiply(A=ParticleRelativeTime(), B=@n18.DetailSpeed)
  n20 = Multiply(A=@n19, B=@n16)
  n21 = VertexNormalWS()
  n22 = ScalarParameter() {parameterName=Vertex_Push}
  Multiply(
   A=TextureSample(UVs=Panner(Coordinate=@n17, Time=@n20) {speedY=-2}) {samplerType=SAMPLERTYPE_LinearGrayscale, texture=/Game/Effects/Fort_Effects/Textures/Tile/T_Motion_Blur_Wind_M}.G
   B=Multiply(A=Multiply(A=@n21, B=@n22), B=VertexColor().R)
  )
  NamedRerouteDeclaration(None=SphereMask(A=TextureCoordinate(), B=Constant() {r=0.5}) {attenuationRadius=0.5, hardnessPercent=0}) {name=EdgeLight, nodeColor=<rgba 0 0.8359 1 1>}
  n23 = AppendVector(A=Constant(), B=ParticleRandom())
  NamedRerouteDeclaration(None=@n23) {name=ParticleUVOffset, nodeColor=<rgba 0.0781 0 1 1>}
  n24 = ComponentMask(None=TextureCoordinate()) {g=true, r=true}
  n25 = LinearInterpolate(A=@n24, B=OneMinus(None=@n24), Alpha=ScalarParameter() {parameterName=FlipWipe})
  n26 = ComponentMask(None=VectorParameter() {parameterName=WipeTiling, defaultValue=<rgba 1 1 0 1>}.RGB) {g=true, r=true}
  n27 = Constant()
  n28 = TextureSample(UVs=Add(A=Multiply(A=@n25, B=@n26), B=AppendVector(A=@n27, B=@n18.Reflections/MaskOffset))) {samplerSource=SSM_Clamp_WorldGroupSettings, texture=/Game/Effects/Fort_Effects/Textures/Masks/T_Square_Gradient_M}
  Reroute(None=@n28.R)
  n29 = ComponentMask(None=TextureCoordinate() {vTiling=4}) {g=true}
  n30 = Multiply(A=@n28.R, B=ParticleColor().A) {constB=2}
  n31 = TextureCoordinate() {uTiling=0.5, vTiling=0.5}
  n32 = Panner(Coordinate=Add(A=Multiply(A=@n31, B=@n18.DetailTiling), B=@n23), Time=@n19) {speedY=-0.75}
  n33 = TextureCoordinate() {vTiling=0.75}
  n34 = Panner(Coordinate=Add(A=Multiply(A=@n33, B=@n18.DetailTiling), B=@n23), Time=@n19) {speedY=-1.5}
  n35 = Add(
   A=Add(
    A=TextureSample(UVs=@n32) {samplerType=SAMPLERTYPE_LinearColor, texture=/Game/Effects/Fort_Effects/Textures/Water/T_waterSplash_D}.G
    B=TextureSample(UVs=@n34) {samplerType=SAMPLERTYPE_LinearColor, texture=&T_waterSplash_D}.G
   )
   B=TextureSample(UVs=Panner(Coordinate=@n15, Time=@n19) {speedY=-2}) {samplerType=SAMPLERTYPE_LinearGrayscale, texture=&T_Motion_Blur_Wind_M}.G
  )
  n36 = DynamicParameter() {
   parameterIndex=2
   defaultValue=<rgba 0 1 1 1>
   paramNames [Reflections/MaskOffset/WhiteWaterOffset, DetailTiling, DetailSpeed, ReflectionsTiling/DetailsTiling/OpacityBoost]
  }
  n37 = Add(A=OneMinus(None=Saturate(None=@n35)), B=@n36.Reflections/MaskOffset/WhiteWaterOffset) {constB=-0.2}
  n38 = Multiply(
   A=LinearInterpolate(B=Saturate(None=@n29), Alpha=ScalarParameter() {defaultValue=1, parameterName=FadeEnd}) {constA=1}
   B=Saturate(None=Subtract(A=@n30, B=@n37))
  )
  n39 = VertexColor()
  Saturate(None=Multiply(A=Multiply(A=@n38, B=@n39.R), B=@n36.ReflectionsTiling/DetailsTiling/OpacityBoost))
 ]
}
