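# Translucent unlit material for Niagara mesh particles, named for the ripple chevrons of the Sunrose Pool effect.
# A noise normal map, panned over particle lifetime with a random per-particle offset, bends a reflection vector that samples tile noise in lat-long UVs as a fake reflection.
# That reflection adds to particle-colored emissive, with dynamic parameters controlling its strength, tiling and detail motion.
# Sphere masks build a soft UV band for alpha and a Fresnel edge mask.
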Material /BattlePassS29_Quests/Items/FX/SunrosePool/M_SunrosePool_RippleChevrons
graph {
 properties {
  bUsedWithNiagaraMeshParticles=true
  blendMode=BLEND_Translucent
  pixelDepthOffsetMode=PDOM_Legacy
  shadingModel=MSM_Unlit
  usedShadingModels=Unlit
 }
 nodes [
  n1 = ComponentMask(None=Divide(A=WorldPosition().XYZ, B=ScalarParameter() {defaultValue=256, parameterName=WorldTextureScale}) {constB=2048}) {g=true, r=true}
  Add(
   A=AppendVector(A=Constant2Vector(), B=DynamicParameter() {paramNames [Reflections, DetailTiling, DetailSpeed, ReflectionsTiling]}.DetailSpeed)
   B=Multiply(
    A=VertexColor().R
    B=TextureSample!(UVs=@n1, Tex, Level, Apply View MipBias) {mipValueMode=TMVM_MipLevel, texture=&T_NoisePack_01}.B
   )
  )
  Constant3Vector() {constant=<rgba 0.2 0.2 0 0>}
  n2 = ComponentMask(None=TextureCoordinate()) {r=true}
  n3 = Multiply(A=Add(A=Multiply(A=@n2) {constB=2}) {constB=0.5}, B=ParticleColor().RGB)
  n4 = OneMinus(None=NamedRerouteUsage())
  n5 = DynamicParameter() {paramNames [Reflections, DetailTiling, DetailSpeed, ReflectionsTiling]}
  n6 = TextureCoordinate() {vTiling=0.25}
  n7 = DynamicParameter() {paramNames [Reflections, DetailTiling, DetailSpeed, ReflectionsTiling]}
  n8 = Panner(Coordinate=Add(A=ParticleRandom(), B=Multiply(A=@n6, B=@n7.DetailTiling)), Time=Multiply(A=ParticleRelativeTime(), B=@n7.DetailSpeed)) {speedY=-1}
  n9 = ScalarParameter() {parameterName=DetailNrm}
  n10 = AppendVector(A=@n9, B=@n9)
  n11 = Constant()
  n12 = Add(
   A=Constant3Vector() {constant=<rgba 0 0 1 0>}
   B=Multiply(
    A=TextureSample!(UVs=@n8, Tex, Apply View MipBias) {samplerType=SAMPLERTYPE_Normal, texture=/Game/Effects/Fort_Effects/Textures/Tile/T_GaussianNoise_02_N}.RGB
    B=AppendVector(A=@n10, B=@n11)
   ) {constB=0}
  )
  n13 = MaterialFunctionCall!(CameraVector=MaterialFunctionCall!(Normal=Transform(None=@n12), CameraVector=CameraVectorWS()) {materialFunction=/Engine/Functions/Engine_MaterialFunctions02/Utility/CustomReflectionVector}.Reflection Vector) {materialFunction=/Engine/Functions/Engine_MaterialFunctions01/Texturing/CameraVectorToLatLongUV}
  n14 = Multiply(
   A=@n5.Reflections
   B=TextureSample(UVs=Multiply(A=@n13.Uvs, B=@n5.ReflectionsTiling) {constB=1.75}) {texture=/Game/Effects/Fort_Effects/Textures/Tile/T_TileNoise_01}.R
  )
  MaterialFunctionCall(Emissive=Add(A=Multiply(A=@n3, B=@n4), B=Add(A=@n3, B=@n14))) {materialFunction=&MF_FX_EmissiveAdjustments}
  n15 = SphereMask(A=ComponentMask(None=TextureCoordinate()) {g=true}, B=Constant() {r=0.5}) {attenuationRadius=0.5, hardnessPercent=0}
  Multiply(A=ParticleColor().A, B=@n15)
  NamedRerouteDeclaration(None=@n15) {name=EdgeLight, nodeColor=<rgba 0 0.8359 1 1>}
  n16 = ScalarParameter() {parameterName=fresnel}
  n17 = StaticBool()
  NamedRerouteDeclaration(None=SphereMask(
    A=MaterialFunctionCall(Power=@n16, Use Cheap contrast=@n17) {materialFunction=&Fresnel_Function}.Result
    B=ScalarParameter() {parameterName=fresnel_contrast_dark}
    Radius=ScalarParameter() {parameterName=fresnel_contrast_bright}
   ) {hardnessPercent=50}) {name=Edge, nodeColor=<rgba 0.7812 1 0 1>}
  NamedRerouteUsage()
 ]
}
